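# Translucent UI material that draws a procedural, texture-free tier indicator shape for the season screens.
# Four rotated, offset UV gradients become hard half-plane edges through a steep multiply and clamp, with a shared Softness scale.
# Scalar dot products intersect the Top and Bottom edges, then cut out the intersection of the two Inset edges.
# A yellow-default Color parameter and a CustomRotator call are present, with no connections shown to the mask.
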Material /Game/Athena/UI/Season/Art/M_UI_TierIndicator
graph {
 properties {
  bUsedWithStaticMesh=false
  blendMode=BLEND_Translucent
  materialDomain=MD_UI
  numCustomizedUVs=1
  pixelDepthOffsetMode=PDOM_Legacy
  usedShadingModels=Default Lit
 }
 nodes [
  n1 = Rotator(Coordinate=TextureCoordinate(), Time=ScalarParameter() {defaultValue=0.16, parameterName=TopAngle}) {speed=6.2832}
  n2 = ScalarParameter() {defaultValue=-0.2, parameterName=TopOffset}
  n3 = OneMinus(None=ComponentMask(None=Add(A=@n1, B=@n2)) {r=true})
  n4 = Constant() {r=0.5}
  n5 = ScalarParameter() {defaultValue=1, parameterName=Softness}
  n6 = Multiply(A=Multiply(A=Subtract(A=@n3, B=@n4)) {constB=80}, B=@n5)
  n7 = Rotator(Coordinate=TextureCoordinate(), Time=ScalarParameter() {defaultValue=-0.135, parameterName=BottomAngle}) {speed=6.2832}
  n8 = ScalarParameter() {defaultValue=-0.2, parameterName=BottomOffset}
  n9 = OneMinus(None=ComponentMask(None=Add(A=@n7, B=@n8)) {r=true})
  n10 = Constant() {r=0.5}
  n11 = Multiply(A=Multiply(A=Subtract(A=@n9, B=@n10)) {constB=80}, B=@n5)
  n12 = Add(
   A=Rotator(Coordinate=TextureCoordinate(), Time=ScalarParameter() {defaultValue=0.075, parameterName=Inset1_Angle}) {speed=6.2832}
   B=ScalarParameter() {defaultValue=-0.05, parameterName=Inset1_Offset}
  )
  n13 = Subtract(A=OneMinus(None=ComponentMask(None=@n12) {r=true}), B=Constant() {r=0.5})
  n14 = Clamp(None=Multiply(A=Multiply(A=@n13) {constB=80}, B=@n5))
  n15 = Add(
   A=Rotator(Coordinate=TextureCoordinate(), Time=ScalarParameter() {defaultValue=-0.05, parameterName=Inset2_Angle}) {speed=6.2832}
   B=ScalarParameter() {defaultValue=-0.05, parameterName=Inset2_Offset}
  )
  n16 = Subtract(A=OneMinus(None=ComponentMask(None=@n15) {r=true}), B=Constant() {r=0.5})
  n17 = Clamp(None=Multiply(A=Multiply(A=@n16) {constB=80}, B=@n5))
  DotProduct(A=DotProduct(A=Clamp(None=@n6), B=Clamp(None=@n11)), B=OneMinus(None=DotProduct(A=@n14, B=@n17)))
  MaterialFunctionCall(UVs=TextureCoordinate(), "Rotation Angle (0-1)"=ScalarParameter() {parameterName=CustomRotation}) {materialFunction=&CustomRotator}
  VectorParameter() {parameterName=Color, defaultValue=<rgba 1 0.95 0 1>}
 ]
}
